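# Shared Fortnite cape AnimBlueprint that animates a cape attached to a character.
# It copies the parent mesh pose, outputting it directly when cloth is enabled, otherwise layering bone dynamics on top.
# That fallback chains trail controllers for spine and forked tips, with aim-driven angular limits and a blended backward pose.
# The event graph updates wind, aim and blend values, and applies Perlin-noise cloth gravity when skydiving with non-negative vertical velocity.
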AnimBlueprint /Figment_S10/Accessories/FORT_Capes/Common/Fortnite_Base_Cape_AnimBlueprint
target_skeleton=/Game/Accessories/FORT_Capes/Common/Fortnite_Base_Cape_Skeleton
graphs [
 AnimGraph {nodes [
   n1 = "Transform (Modify) Bone - Bone: backpack_root"(
    ComponentPose=Local To Component()
    Alpha=Get swing_rotate_alpha() -> (swing_rotate_alpha:double)
    Translation="0,0,0"
    Rotation="50,0,30"
    Scale="1,1,1"
   ) {
    rotationMode=BMM_Additive
    rotationSpace=BCS_BoneSpace
    alphaBoolBlend {blendOption=Linear}
    alphaScaleBiasClamp {bInterpResult=true, interpSpeedDecreasing=5}
    boneToModify {boneName=backpack_root}
    rotation {pitch=50, roll=30}
   }
   Component To Local(ComponentPose=@n1)
   Local Space Ref Pose()
   n2 = Get bIsTargeting()
   n3 = Trail controller - Bone: dyn_cape05!(
    ComponentPose=Local To Component(LocalPose=Use cached pose 'Cape_Cache'())
    Alpha=SelectFloat(A=SelectFloat(A=0, B=0.25, bPickA=@n2), B=1, bPickA=Get bIsCrouching())
    FakeVelocity:Vector="0,0,0"
   ) -> (Pose) {
    bActorSpaceFakeVel=true
    bUsePlanarLimit=true
    chainLength=5
    lastBoneRotationAnimAlphaBlend=0.5
    maxDeltaTime=0.0333
    baseJoint {boneName=backpack_root}
    fakeVelocity {y=50, z=-50}
    planarLimits [
     {drivingBone {boneName=backpack_root}, planeTransform {location=<0 -2.0567 0>, rotation=<rot 0 0 -90>, scale=<1 1 1>}}
     {drivingBone {boneName=backpack_root}, planeTransform {location=<0 0 10>, rotation=<rot 0 0 -150>, scale=<1 1 1>}}
    ]
    rotationLimits [{limitMax=<180 180 180>}, {limitMax=<180 180 180>}, {limitMax=<180 180 180>}, {limitMax=<180 180 180>}, {limitMax=<180 180 180>}]
    rotationOffsets [<0 0 0>, <0 0 0>, <0 0 0>, <0 0 0>, <0 0 0>]
    trailBone {boneName=dyn_cape05}
    trailRelaxationSpeed {editorCurveData {keys [<at 0 20 RCIM_Linear>, {arriveTangent=-14.4971, interpMode=RCIM_Linear, leaveTangent=-14.4971, time=1, value=10}]}}
   }
   n4 = Apply Limits!(ComponentPose:Component Space Pose Link Structure=@n3, Alpha:float=1, AngularOffsets_0:Vector=Get AimOffset() -> (AimOffset:Vector)) -> (Pose:Component Space Pose Link Structure) {
    is=title AnimGraphNode_ApplyLimits
    alphaBoolBlend {blendOption=Linear}
    angularOffsets [<0 0 0>]
    angularRangeLimits [{bone {boneName=dyn_cape01}, limitMax=<10 180 0>, limitMin=<-10 -180 -120>}]
   }
   n5 = Component To Local(ComponentPose=@n4)
   n6 = Fortnite_Base_Cape_PoseAsset : BackwardPose!(PoseAsset:Pose Asset Object Reference=&/Fortnite_Base_Cape_PoseAsset) -> (Pose:Pose Link Structure) {is=title AnimGraphNode_PoseByName, poseName=BackwardPose}
   n7 = Get bBlendInBackwardPose() -> (bBlendInBackwardPose:bool)
   n8 = Get FinalCapeBlendOutAlpha() -> (FinalCapeBlendOutAlpha:float)
   n9 = Get FinalCapeWindVelocity() -> (FinalCapeWindVelocity:Vector)
   n10 = Trail controller - Bone: dyn_cape06_l!(
    ComponentPose=Local To Component(LocalPose=Two Way Blend!(A=@n5, B=@n6, bAlphaBoolEnabled:bool=@n7) -> (Pose) {blendNode {alphaInputType=Bool, alphaBoolBlend {blendInTime=0.5, blendOption=Linear, blendOutTime=0.5}}})
    Alpha=@n8
    FakeVelocity:Vector=@n9
   ) -> (Pose) {
    bActorSpaceFakeVel=true
    chainLength=6
    lastBoneRotationAnimAlphaBlend=1
    maxDeltaTime=0.0333
    baseJoint {boneName=backpack_root}
    planarLimits [{drivingBone {boneName=dyn_cape05}, planeTransform {location=<0 -2.0567 0>, rotation=<rot 0 0 -90>, scale=<1 1 1>}}]
    rotationLimits [
     {limitMax=<180 180 180>}
     {limitMax=<180 180 180>}
     {limitMax=<180 180 180>}
     {limitMax=<180 180 180>}
     {limitMax=<180 180 180>}
     {limitMax=<180 180 180>}
    ]
    rotationOffsets [<0 0 0>, <0 0 0>, <0 0 0>, <0 0 0>, <0 0 0>, <0 0 0>]
    trailBone {boneName=dyn_cape06_l}
    trailRelaxationSpeed {editorCurveData {keys [<at 0 15 RCIM_Linear>,<at 1 6 RCIM_Linear>]}}
   }
   n11 = Trail controller - Bone: dyn_cape06_r!(ComponentPose=@n10, Alpha=@n8, FakeVelocity:Vector=@n9) -> (Pose) {
    bActorSpaceFakeVel=true
    chainLength=6
    lastBoneRotationAnimAlphaBlend=1
    maxDeltaTime=0.0333
    baseJoint {boneName=backpack_root}
    planarLimits [{drivingBone {boneName=dyn_cape05}, planeTransform {location=<0 -2.0567 0>, rotation=<rot 0 0 -90>, scale=<1 1 1>}}]
    rotationLimits [
     {limitMax=<180 180 180>}
     {limitMax=<180 180 180>}
     {limitMax=<180 180 180>}
     {limitMax=<180 180 180>}
     {limitMax=<180 180 180>}
     {limitMax=<180 180 180>}
    ]
    rotationOffsets [<0 0 0>, <0 0 0>, <0 0 0>, <0 0 0>, <0 0 0>, <0 0 0>]
    trailBone {boneName=dyn_cape06_r}
    trailRelaxationSpeed {editorCurveData {keys [<at 0 15 RCIM_Linear>,<at 1 8 RCIM_Linear>]}}
   }
   n12 = Get bCanPlayCustomAnimations() -> (bCanPlayCustomAnimations:bool)
   n13 = "Transform (Modify) Bone - Bone: dyn_cape01"!(ComponentPose=@n11, Alpha="To Float (Boolean)"(InBool=NOT Boolean(A=@n12)), Rotation="0,0,-10") -> (Pose) {rotationMode=BMM_Additive, alphaBoolBlend {blendOption=Linear}, boneToModify {boneName=dyn_cape01}, rotation {roll=-10}}
   Output Pose(Result=Blend Poses by bool(
     BlendPose_0=Use cached pose 'Cape_Cache'()
     BlendPose_1=Component To Local(ComponentPose=@n13)
     BlendTime_0=0
     BlendTime_1=0
     bActiveValue=Get bClothEnabled()
    ) {blendType=Linear}) {layerGroup=None}
   Save cached pose 'Cape_Cache'(Pose=Copy Pose From Mesh() {bUseAttachedParent=true}) {cacheName=Cape_Cache}
   AnimGraphNode_SubInput_0 = Input Pose()
   n14 = Get bIsDBNO()
   SelectFloat(A=SelectFloat(A=0.5, B=Get BlendOutForksAlpha() -> (BlendOutForksAlpha:float), bPickA=@n14), B=1, bPickA=@n7)
   SelectVector(A=Get FakeWindVelocityDBNO() -> (FakeWindVelocityDBNO:Vector), B=Get FakeWindVelocity() -> (FakeWindVelocity:Vector), bPickA=@n14)
  ]}
 EventGraph {nodes [
   Event BlueprintUpdateAnimation()
   n1 = BreakTransform(InTransform=GetDeltaTransformFromRefPose(self=GetOwnerSkeletalMeshComponent(), BoneName=upperarm_twist_01_l, BaseName=clavicle_l))
   > Set swing_rotate_alpha(swing_rotate_alpha:double=MapRangeClamped(Value=BreakRotator(InRot=@n1.Rotation).Yaw, InRangeA=10, InRangeB=-40)) -> (Output_Get:double)
   > Branch(Condition=OR Boolean(A=Get bIsSkyDiving(), B=Get bCanPlayCustomAnimations()))
   > Set swing_rotate_alpha(swing_rotate_alpha=0)
   Reroute Node()
   > Set isDBNO_Old(isDBNO_Old:bool=IsDBNO!(self:Fort Pawn Object Reference=TryGetFortPawn()) -> (ReturnValue:bool)) -> (Output_Get:bool)
   > Set FakeWindVelocity_DBNO_Old(FakeWindVelocity_DBNO_Old:Vector="0,0,100") -> (Output_Get:Vector)
   n2 = "Absolute (Float)"(A=GetOwnerDirection!(self:Custom Character Part Anim Instance Object Reference) -> (ReturnValue:float))
   n3 = NormalizeToRange!(Value:double=@n2, RangeMin:double=0, RangeMax:double=180) -> (ReturnValue:double)
   > Set BlendOutForksAlpha(BlendOutForksAlpha:float=float - float(B="Clamp (Float)"(Value=@n3, Min=0, Max=1)) {is=title K2Node_CallFunction}) -> (Output_Get:float)
   > Set bBlendInBackwardPose(bBlendInBackwardPose:bool=OR Boolean(A=Get bIsDBNO(), B="float > float"(A=@n2, B=100) {is=title K2Node_CallFunction})) -> (Output_Get:bool)
   n4 = "Clamp (Float)"(Value=Get PawnSpeed2D(), Min=0.5, Max=1)
   > Set FakeWindVelocity(FakeWindVelocity:Vector=vector * vector(A:Vector="To Vector (Float)"!(InDouble:double=@n4) -> (ReturnValue:Vector), B:Vector=Get CapeFakeVelocity_0() -> (CapeFakeVelocity_0:Vector)) -> (ReturnValue:Vector) {is=title K2Node_CommutativeAssociativeBinaryOperator}) -> (Output_Get:Vector)
   n5 = Get AimPitch()
   n6 = Get AimOffsetMultiplier() -> (AimOffsetMultiplier:float)
   > Set!(
    execute
    AimOffset_X:double=0
    AimOffset_Y:double=0
    AimOffset_Z:double="Min (Float)"(A:double=float * float(A=@n5, B=@n6), B:double=0) -> (ReturnValue:double) {is=title K2Node_CommutativeAssociativeBinaryOperator}
   ) -> (then, Output_Get:Vector)
   > n7 = Branch(Condition="float >= float"(A=Get PawnVelocityZ(), B=0) {is=title K2Node_CallFunction})
   > n8 = Branch(Condition=Get bIsSkyDiving())
   n9 = MapRangeClamped(Value=GetWorldDeltaSeconds!() -> (ReturnValue:double), InRangeA=0.02, InRangeB=0.03, OutRangeA=-100, OutRangeB=100)
   n10 = GetActorForwardVector!(self:Actor Object Reference=TryGetPawnOwner()) -> (ReturnValue:Vector)
   n11 = Get ClothingSimInteractor() -> (ClothingSimInteractor:Clothing Simulation Interactor Object Reference)
   > n12 = EnableGravityOverride!(
    execute
    self:Clothing Simulation Interactor Object Reference=@n11
    InVector:"Vector (by ref)"=vector * vector(
     A=vector + vector(A:Vector=@n10, B:Vector="0, 0,10") -> (ReturnValue:Vector) {is=title K2Node_CommutativeAssociativeBinaryOperator}
     B=Make Perlin Noise Vector and Remap!(
      X:float=@n9
      Y:float=@n9
      Z:float=@n9
      RangeOutMinX:float=-20000
      RangeOutMaxX:float=-5000
      RangeOutMinY:float=-20000
      RangeOutMaxY:float=-5000
      RangeOutMinZ:float=100
      RangeOutMaxZ:float=200
     ) -> (ReturnValue:Vector)
    )
   ) -> (then)
   ForceClothNextUpdateTeleportAndReset(
    execute=[@n12.then, @n13.Completed]
    self:Skeletal Mesh Component Object Reference=Get MeshToCopy() -> (MeshToCopy:Skeletal Mesh Component Object Reference)
   )
   DisableGravityOverride(execute=[@n8.else, @n7.else], self:Clothing Simulation Interactor Object Reference=@n11)
   > n13 = Do Once()
   Reroute Node()
   > n14 = Cast To ClothingSimulationInteractorNv!(
    execute
    Object=GetClothingSimulationInteractor!(self:Skeletal Mesh Component Object Reference=GetPartSkeletalMeshComponent(), ClothingAssetName:Name=None) -> (ReturnValue:Clothing Simulation Interactor Object Reference)
   ) -> (then, CastFailed, AsClothing Simulation Interactor Nv:Clothing Simulation Interactor Nv Object Reference) {is=title K2Node_DynamicCast}
   > Set clothing_sim_interactor_Old(clothing_sim_interactor_Old:Clothing Simulation Interactor Nv Object Reference=@n14) -> (Output_Get:Clothing Simulation Interactor Nv Object Reference)
   > Set MeshToCopy_0(MeshToCopy_0:Skeletal Mesh Component Object Reference=GetOwnerSkeletalMeshComponent()) -> (Output_Get:Skeletal Mesh Component Object Reference)
   Event BlueprintInitializeAnimation()
   Get clothing_sim_interactor_Old() -> (clothing_sim_interactor_Old:Clothing Simulation Interactor Nv Object Reference)
   Get PawnMovementDirectionAzimuth() -> (PawnMovementDirectionAzimuth:float)
   Get MeshToCopy_0() -> (MeshToCopy_0:Skeletal Mesh Component Object Reference)
   Get bCanPlayCustomAnimations()
  ]}
]
variables [
 ActorReverse {type=boolean}
 ActorVelocity {type=number}
 AimOffset_Old {type=Vector}
 AimOffsetMultiplier_Old {type=number, default=-0.25}
 BlendInBackwardPose_Old {type=boolean}
 BlendOutForksAlpha_Old {type=number}
 CapeFakeVelocity_0 {type=Vector, default=<30 30 -50>}
 clothing_sim_interactor_Old {type=ClothingSimulationInteractorNv, default=null}
 FakeWindVelocity_DBNO_Old {type=Vector}
 FakeWindVelocity_Old {type=Vector}
 Forward {type=number}
 isDBNO_Old {type=boolean}
 MeshToCopy_0 {type=SkeletalMeshComponent}
 Side {type=number}
 swing_rotate_alpha {type=number}
 TrailAlphaBlend {type=number}
]
